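# Mobile grass material function for biome landscapes, with separate outputs for landscape material, static meshes and height.
# It wraps MF_Common_Grass, tinting landscape base color by the combined layers mask and both colors toward a cliff-edge color using MF_Mobile_Common_Cliff's soft slope.
# Virtual texture and quality switches adjust or bypass that slope blend.
# A shading path switch gives forward and mobile meshes constant specular, roughness and flat normal.
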MaterialFunction /Clyde/Athena/Apollo/Environments/Landscape/Materials/Biome/CommonMaterials/Mobile/MF_Mobile_Common_Grass
graph {nodes [
  FunctionInput(Preview=Constant()) {bUsePreviewValueAsDefault=true, inputName=Slope-CliffMask, inputType=FunctionInput_Scalar}
  n1 = VectorParameter() {parameterName=Mobile-CliffEdgeColor, defaultValue=<rgba 0.6099 0.7 0 1>}
  n2 = MaterialFunctionCall!(BiomeMask, UseBiomeColorMask, CliffMask, Grass-EdgeBiomeColorA, Grass-EdgeBiomeColorB, CombinedLandscapeLayersMask) {materialFunction=/Clyde/Athena/Apollo/Environments/Landscape/Materials/Biome/CommonMaterials/MF_Common_Grass}
  n3 = LinearInterpolate(
   A=@n2.Mobile-Basecolor
   B=Multiply(A=VectorParameter() {parameterName=Mobile-GrassEdgeColor, defaultValue=<rgba 1.65 1.2465 0.0825 1>}.RGB, B=@n2.Mobile-Basecolor)
   Alpha=FunctionInput(Preview=Constant()) {bUsePreviewValueAsDefault=true, inputName=CombinedLandscapeLayersMask, inputType=FunctionInput_Scalar}
  )
  n4 = MaterialFunctionCall!(Gravel) {materialFunction=&/MF_Mobile_Common_Cliff}
  n5 = Constant() {r=1}
  n6 = ScalarParameter() {parameterName=d4trgeyyhe}
  n7 = Saturate(None=VirtualTextureFeatureSwitch!(No=QualitySwitch(Default=@n4.SoftSlope, Low=@n5, High=@n5, Medium=@n5), Yes=Add(A=@n4.SoftSlope, B=@n6)))
  n8 = Constant() {r=0.3}
  n9 = Constant() {r=0.9}
  n10 = SetMaterialAttributes!(
   MaterialAttributes
   BaseColor=LinearInterpolate(A=Multiply(A=@n1.RGB, B=@n3), B=@n3, Alpha=@n7)
   Specular=@n8
   Roughness=@n9
   Normal=VertexNormalWS()
  ) {attributeSetTypes [
    #1
    #2
    #3
    #4
   ]}
  FunctionOutput(None=@n10) {desc=Mobile Landscape Only, outputName=Material, sortPriority=-1}
  n11 = QualitySwitch(Default=@n2.Basecolor, Low=@n2.Basecolor, Medium=@n2.Basecolor) {desc="Mobile was plugged into High/default - swapped around\r\nto make grass consistent/nblend on all platforms"}
  n12 = Saturate(None=@n11)
  n13 = LinearInterpolate(A=Multiply(A=@n1.RGB, B=@n12), B=@n12, Alpha=@n7)
  n14 = SetMaterialAttributes(BaseColor=@n13, Specular=@n2.SpecTexture-NoRVT, Roughness=@n2.SpecTexture-NoRVT, Normal=@n2.SpecTexture-NoRVT) {attributeSetTypes [
    #1
    #2
    #3
    #4
   ]}
  n15 = SetMaterialAttributes(BaseColor=@n13, Specular=@n8, Roughness=@n9, Normal=Constant3Vector() {constant=<rgba 0 0 1 0>}) {attributeSetTypes [
    #1
    #2
    #3
    #4
   ]}
  FunctionOutput(None=ShadingPathSwitch(Default=@n14, Deferred=@n14, Forward=@n15, Mobile=@n15)) {outputName=Output_for_StaticMesh, sortPriority=80}
  FunctionOutput(None=@n2.Mobile-Height) {outputName=Height}
  n16 = MaterialFunctionCall!(CombinedLandscapeLayersMask, BypassUseRVT, BypassBiomeColors) {materialFunction=/Game/Athena/Apollo/Environments/Landscape/Materials/Biome/CommonMaterials/MF_Common_Grass}
  Reroute(None=@n16.BasicGrassMeshColor)
  Reroute(None=@n16.Spec-CameraDistance)
  Reroute(None=@n16.Rough-Basic)
  Reroute(None=@n16.TangentNormals)
  Reroute(None=@n16.Mobile-Height)
  VirtualTextureFeatureSwitch(No=@n2.Mobile-Basecolor, Yes=@n11) {desc="Disconnected, this is handled in Common_grass"}
 ]}
