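# Material function providing the common mobile sand layer for Apollo biome landscapes and static meshes, using Z-projected world-aligned textures.
# The mobile landscape output uses diffuse color, halved on forward and mobile paths, with constant specular and roughness and the vertex normal.
# The static mesh output adds a normal map and drives specular and roughness from spec red, falling back to the simple version on forward and mobile.
# Diffuse alpha is output as Height.
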MaterialFunction /Clyde/Athena/Apollo/Environments/Landscape/Materials/Biome/CommonMaterials/Mobile/MF_Mobile_Common_Sand
graph {nodes [
  n1 = Constant() {r=550}
  n2 = StaticBool()
  n3 = MaterialFunctionCall!(
   Diffuse Texture Object=TextureObject() {texture=/Game/Environments/Apollo/Sets/Terrain/Textures/T_Apollo_Sand_D_01}
   Spec Texture Object=TextureObject() {samplerType=SAMPLERTYPE_Masks, texture=/Game/Environments/Apollo/Sets/Terrain/Textures/T_Apollo_Sand_S_01}
   TextureSize=@n1
   --------------------------
   Use seperate settings for Z projection
   Use seperate texture for Z projection
   Texture Size Z projection
   Z Plane Diffuse Texture Object
   Z Plane Spec Texture Object
   "--------------------------      "
   ProjectionTransitionContrast
   Use Texture Alpha Heightmap to Refine the Transistion=StaticBool()
   HeightLerpTransitionContrast
   PickTheDiffuseHeightChannel
   TheDiffuseHeightChannel
   "--------------------------       "
   NormalTextureObject
   World Space Normals=@n2
   "--------------------------  "
   Export Float 4 Diffuse=StaticBool() {value=true}
   Export Float 4 Spec
   "--------------------------     "
   WorldPosition=WorldPosition() {worldPositionShaderOffset=WPT_ExcludeAllShaderOffsets}.XYZ
   World Space Normal
  ) {materialFunction=/Engine/Functions/Engine_MaterialFunctions02/Texturing/WorldAlignedTextures_Complex}
  n4 = ComponentMask(None=@n3.Z Diffuse Texture) {b=true, g=true, r=true}
  n5 = Multiply(A=Constant() {r=0.5}, B=@n4)
  n6 = ShadingPathSwitch(Default=@n4, Deferred=@n4, Forward=@n5, Mobile=@n5)
  n7 = SetMaterialAttributes!(MaterialAttributes, BaseColor=@n6, Specular=Constant() {r=0.3}, Roughness=Constant() {r=0.9}, Normal=VertexNormalWS()) {attributeSetTypes [
    #1
    #2
    #3
    #4
   ]}
  FunctionOutput(None=@n7) {desc=Mobile Landscape Only, outputName=Material, sortPriority=-1}
  n8 = MaterialFunctionCall(Float3=@n3.Z Spec) {materialFunction=&BreakOutFloat3Components}
  n9 = SetMaterialAttributes(
   BaseColor=@n6
   Specular=@n8.R
   Roughness=@n8.R
   Normal=MaterialFunctionCall(
    TextureObject=TextureObject() {samplerType=SAMPLERTYPE_Normal, texture=/Game/Environments/Apollo/Sets/Terrain/Textures/T_Apollo_Sand_N_01}
    TextureSize=@n1
    WorldSpace=@n2
   ) {materialFunction=&WorldAlignedNormal}.Z Texture
  ) {attributeSetTypes [
    #1
    #2
    #3
    #4
   ]}
  FunctionOutput(None=ShadingPathSwitch(Default=@n9, Forward=@n7, Mobile=@n7)) {outputName=OutPut_Staticmesh, sortPriority=6}
  FunctionOutput(None=ComponentMask(None=@n3.Z Diffuse Texture) {a=true}) {outputName=Height}
 ]}
